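# Physics asset for the dangling skull-and-teeth necklace on the medium female Dino Collector character body.
# Capsule bodies form a chain from a small master bone through two necklace links, branching at the second into the skull and four teeth.
# Links and skull have tight angular limits, while the teeth get wider swing.
# Large spine_04 and spine_05 capsules provide torso volume, with free angular constraints between them and to the necklace master.
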PhysicsAsset /Game/Characters/Player/Female/Medium/Bodies/F_MED_Dino_Collector/Meshes/F_MED_Dino_Collector_dyn_necklace
bodies {
 dyn_necklace_01 [{shapeType=Capsule, length=5.625, radius=2.375, center=<9.6684 0 1.2167>, rotation=<rot 86.0467 0 0>}]
 dyn_necklace_02 [{shapeType=Capsule, length=5.625, radius=2.375, center=<6.1577 0.0002 0.6679>, rotation=<rot 84.4184 0 0>}]
 dyn_necklace_MSTR [{shapeType=Capsule, length=0.5, radius=0.5}]
 dyn_skull [{shapeType=Capsule, length=5.5, radius=4.875, center=<1.2469 0 4.2529>, rotation=<rot 88.1169 179.9998 179.9998>}]
 dyn_teeth_01 [{shapeType=Capsule, length=1.7659, radius=0.9463, center=<1.6753 -0.2166 -0.3136>, rotation=<rot 0 -96.5769 -81.5812>}]
 dyn_teeth_02 [{shapeType=Capsule, length=2.1844, radius=0.9572, center=<2.1006 -0.0791 -0.2779>, rotation=<rot 0 -92.6058 -83.0929>}]
 dyn_teeth_03 [{shapeType=Capsule, length=2.091, radius=0.9546, center=<2.4983 0.1165 -0.3611>, rotation=<rot 0 -89.2974 -83.5408>}]
 dyn_teeth_04 [{shapeType=Capsule, length=1.7687, radius=0.9488, center=<1.6055 0.1463 -0.0439>, rotation=<rot 0 -81.7284 -88.7404>}]
 spine_04 [{shapeType=Capsule, length=16.9509, radius=9.2111, center=<-0.8885 5.031 0>, rotation=<rot 0 90 0>}]
 spine_05 [{shapeType=Capsule, length=15.5759, radius=10.7111, center=<-3.2831 -1.006 0>, rotation=<rot 0 90 0>}]
}
constraints [
 {
  bone1Name=dyn_necklace_MSTR
  bone2Name=spine_05
  swing1LimitDegrees=45
  swing1Motion=Free
  swing2LimitDegrees=45
  swing2Motion=Free
  twistLimitDegrees=45
  twistMotion=Free
 }
 {bone1Name=dyn_necklace_01, bone2Name=dyn_necklace_MSTR, swing1LimitDegrees=2, swing2LimitDegrees=10, twistLimitDegrees=2}
 {bone1Name=dyn_necklace_02, bone2Name=dyn_necklace_01, swing1LimitDegrees=2, swing2LimitDegrees=10, twistLimitDegrees=2}
 {bone1Name=dyn_skull, bone2Name=dyn_necklace_02, swing1LimitDegrees=2, swing2LimitDegrees=10, twistLimitDegrees=2}
 {bone1Name=dyn_teeth_01, bone2Name=dyn_necklace_02, swing1LimitDegrees=15, swing2LimitDegrees=15, twistLimitDegrees=2}
 {bone1Name=dyn_teeth_02, bone2Name=dyn_necklace_02, swing1LimitDegrees=15, swing2LimitDegrees=15, twistLimitDegrees=2}
 {bone1Name=dyn_teeth_03, bone2Name=dyn_necklace_02, swing1LimitDegrees=15, swing2LimitDegrees=15, twistLimitDegrees=2}
 {bone1Name=dyn_teeth_04, bone2Name=dyn_necklace_02, swing1LimitDegrees=15, swing2LimitDegrees=15, twistLimitDegrees=2}
 {
  bone1Name=spine_05
  bone2Name=spine_04
  swing1LimitDegrees=45
  swing1Motion=Free
  swing2LimitDegrees=45
  swing2Motion=Free
  twistLimitDegrees=45
  twistMotion=Free
 }
]
