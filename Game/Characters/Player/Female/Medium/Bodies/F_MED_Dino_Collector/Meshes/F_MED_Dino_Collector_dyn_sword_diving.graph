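# Physics asset for the female medium Dino Collector player body, a diving variant (by name) for dynamic sword, hip hanging and cloth bones.
# Capsules on pelvis, spine_03 and left thigh and calf provide body shapes, linked to the pelvis by free angular constraints.
# A two-segment sword chain and two hip pieces use tight angular limits below pelvis-linked MSTR anchors.
# Three cloth bones each attach directly to spine_03, swinging mainly on one axis.
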PhysicsAsset /Game/Characters/Player/Female/Medium/Bodies/F_MED_Dino_Collector/Meshes/F_MED_Dino_Collector_dyn_sword_diving
bodies {
 calf_l [{shapeType=Capsule, length=41.5289, radius=6.2064, center=<-20.4588 -2.9089 -0.1193>, rotation=<rot 4.3781 90.9129 -88.791>}]
 dyn_cloth_01 [{shapeType=Capsule, length=3.6747, radius=0.9247, center=<1.2543 0 0>, rotation=<rot 0 0 -90>}]
 dyn_cloth_02 [{shapeType=Capsule, length=3.4163, radius=1.0628, center=<2.6889 0 0>, rotation=<rot 0 0 90>}]
 dyn_cloth_03 [{shapeType=Capsule, length=1.7758, radius=1.0646, center=<1.4086 0 0>, rotation=<rot 0 0 90>}]
 dyn_hip_hanging_01 [{shapeType=Capsule, length=4.25, radius=2.75, center=<10 0 1>, rotation=<rot 0 0 90>}]
 dyn_hip_hanging_01_MSTR [{shapeType=Capsule, length=0.505, radius=0.505, rotation=<rot 0 -90 0>}]
 dyn_hip_hanging_02 [{shapeType=Capsule, length=4.25, radius=2.75, center=<10 0 1.5>, rotation=<rot 0 0 85.1379>}]
 dyn_hip_hanging_02_MSTR [{shapeType=Capsule, length=0.505, radius=0.505, rotation=<rot 0 -90 0>}]
 dyn_sword_01 [{shapeType=Capsule, length=7.5, radius=2.5, center=<10 0 0>, rotation=<rot 0.0001 17.4251 79.8132>}]
 dyn_sword_02 [{shapeType=Capsule, length=69, radius=3.25, center=<9.8383 -3.2781 0.0001>, rotation=<rot 0 1.824 89.9999>}]
 dyn_sword_MSTR [{shapeType=Capsule, length=0.505, radius=0.505, rotation=<rot 0 -90 0>}]
 pelvis [{shapeType=Capsule, length=7.9273, radius=11.6003, center=<0.5295 -1.2521 -0.2626>}]
 spine_03 [{shapeType=Capsule, length=3.6773, radius=7.6003, center=<-2.4372 3.096 -0.2626>}]
 thigh_twist_01_l [{shapeType=Capsule, length=32.6089, radius=8.9282, center=<9.8907 -2.4385 2.4914>, rotation=<rot 86.2941 0 0>}]
}
constraints [
 {
  bone1Name=thigh_twist_01_l
  bone2Name=pelvis
  swing1LimitDegrees=45
  swing1Motion=Free
  swing2LimitDegrees=45
  swing2Motion=Free
  twistLimitDegrees=45
  twistMotion=Free
 }
 {
  bone1Name=dyn_sword_MSTR
  bone2Name=pelvis
  swing1LimitDegrees=45
  swing1Motion=Free
  swing2LimitDegrees=45
  swing2Motion=Free
  twistLimitDegrees=45
  twistMotion=Free
 }
 {bone1Name=dyn_sword_01, bone2Name=dyn_sword_MSTR, swing1LimitDegrees=2, swing2LimitDegrees=10, twistLimitDegrees=2}
 {bone1Name=dyn_sword_02, bone2Name=dyn_sword_01, swing1LimitDegrees=15, swing2LimitDegrees=10, twistLimitDegrees=10}
 {
  bone1Name=calf_l
  bone2Name=pelvis
  swing1LimitDegrees=45
  swing1Motion=Free
  swing2LimitDegrees=45
  swing2Motion=Free
  twistLimitDegrees=45
  twistMotion=Free
 }
 {
  bone1Name=dyn_hip_hanging_01_MSTR
  bone2Name=pelvis
  swing1LimitDegrees=45
  swing1Motion=Free
  swing2LimitDegrees=45
  swing2Motion=Free
  twistLimitDegrees=45
  twistMotion=Free
 }
 {bone1Name=dyn_hip_hanging_01, bone2Name=dyn_hip_hanging_01_MSTR, swing1LimitDegrees=2, swing2LimitDegrees=25, twistLimitDegrees=2}
 {
  bone1Name=dyn_hip_hanging_02_MSTR
  bone2Name=pelvis
  swing1LimitDegrees=45
  swing1Motion=Free
  swing2LimitDegrees=45
  swing2Motion=Free
  twistLimitDegrees=45
  twistMotion=Free
 }
 {bone1Name=dyn_hip_hanging_02, bone2Name=dyn_hip_hanging_02_MSTR, swing1LimitDegrees=2, swing2LimitDegrees=10, twistLimitDegrees=2}
 {bone1Name=dyn_cloth_01, bone2Name=spine_03, swing1LimitDegrees=5, swing2LimitDegrees=60, twistLimitDegrees=2}
 {bone1Name=dyn_cloth_02, bone2Name=spine_03, swing1LimitDegrees=5, swing2LimitDegrees=60, twistLimitDegrees=2}
 {bone1Name=dyn_cloth_03, bone2Name=spine_03, swing1LimitDegrees=5, swing2LimitDegrees=60, twistLimitDegrees=2}
 {
  bone1Name=spine_03
  bone2Name=pelvis
  swing1LimitDegrees=45
  swing1Motion=Free
  swing2LimitDegrees=45
  swing2Motion=Free
  twistLimitDegrees=45
  twistMotion=Free
 }
]
