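# Post-process-style AnimBP for the F_MED_Dino_Collector player body, copying the parent mesh pose and animating its sword, necklace and dangling accessories.
# Two deform Control Rigs, constraints, angle-driven corrective offsets and Anim Dynamics build a cached pre-pose.
# RigidBody nodes then simulate the necklace and sword, with a skydiving variant, or Look Ats pose the sword when rigid bodies are disabled.
# The EventGraph derives corrective alphas from leg rotations, movement and dive state.
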AnimBlueprint /Game/Characters/Player/Female/Medium/Bodies/F_MED_Dino_Collector/Meshes/F_MED_Dino_Collector_AnimBP
target_skeleton=&/F_MED_Dino_Collector_Skeleton
graphs [
 AnimGraph {nodes [
   n1 = Look At - Bone: dyn_sword_01(ComponentPose=Local To Component(LocalPose=Use cached pose 'prePose'()), Alpha=1) {
    alphaBoolBlend {blendOption=Linear}
    boneToModify {boneName=dyn_sword_01}
    lookAtLocation=<0 5 -35>
    lookAtTarget {boneReference {boneName=calf_l}}
    lookAt_Axis {axis=<1 0>}
    lookUp_Axis {axis {x=0}}
   }
   n2 = Look At - Bone: dyn_sword_02(ComponentPose=@n1, Alpha=1) {
    alphaBoolBlend {blendOption=Linear}
    boneToModify {boneName=dyn_sword_02}
    lookAtLocation=<0 10 -8>
    lookAtTarget {boneReference {boneName=calf_l}}
    lookAt_Axis {axis=<1 0>}
    lookUp_Axis {axis {x=0}}
   }
   n3 = "Transform (Modify) Bone - Bone: dyn_sword_02"(ComponentPose=@n2, Alpha=Get isSkyDiving() -> (isSkyDiving:double), Translation="0,0,0", Rotation="10,40,0", Scale="1,1,1") {
    rotationMode=BMM_Additive
    rotationSpace=BCS_BoneSpace
    scaleMode=BMM_Additive
    scaleSpace=BCS_BoneSpace
    translationMode=BMM_Additive
    translationSpace=BCS_BoneSpace
    alphaBoolBlend {blendOption=Linear}
    alphaScaleBiasClamp {bInterpResult=true}
    boneToModify {boneName=dyn_sword_02}
    rotation {pitch=10, yaw=40}
   }
   n4 = Component To Local(ComponentPose=@n3)
   n5 = Look At - Bone: dyn_sword_01(ComponentPose=Local To Component(LocalPose=Use cached pose 'prePose'()), Alpha=1) {
    alphaBoolBlend {blendOption=Linear}
    boneToModify {boneName=dyn_sword_01}
    lookAtLocation=<85 10>
    lookAtTarget {boneReference {boneName=root}}
    lookAt_Axis {axis=<1 0>}
    lookUp_Axis {axis {x=0}}
   }
   n6 = Look At - Bone: dyn_sword_02(ComponentPose=@n5, Alpha=1) {
    alphaBoolBlend {blendOption=Linear}
    boneToModify {boneName=dyn_sword_02}
    lookAtLocation=<30 30>
    lookAtTarget {boneReference {boneName=root}}
    lookAt_Axis {axis=<1 0>}
    lookUp_Axis {axis {x=0}}
   }
   n7 = Component To Local(ComponentPose=@n6)
   n8 = Get bIsOnGround() -> (bIsOnGround:bool)
   Output Pose(Result=Blend Poses by bool(
     BlendPose_0=Use cached pose 'rigidCache'()
     BlendPose_1=Blend Poses by bool(BlendPose_0=@n7, BlendPose_1=@n4, bActiveValue=@n8) {blendType=Linear}
     bActiveValue=Get bRigidBodyEnabled()
    ) {blendType=Linear}) {layerGroup=None}
   n9 = Control Rig(Source=Copy Pose From Mesh() {bUseAttachedParent=true}) {
    alphaBoolBlend {blendOption=Linear}
    controlRigAssetReference {blueprintRigClass=&/F_MED_Dino_Collector_Deform_CtrlRig_twists.F_MED_Dino_Collector_Deform_CtrlRig_twists_C}
   }
   n10 = Control Rig!(
    Source=@n9
    Alpha
    KneeRot_fr_l_114deg:Vector
    KneeTrans_fr_l_114deg:Vector="0.258,-3.591,0.009"
    KneeScale_fr_l_114deg:Vector="1,1,1"
    KneeScale_fr_l_130deg:Vector="1,1,1"
    KneeRot_fr_l_130deg:Vector="0,0,-51.2"
    KneeTrans_fr_l_130deg:Vector="0.258,-3.591,0.009"
   ) -> (Pose) {
    alphaBoolBlend {blendOption=Linear}
    controlRigAssetReference {blueprintRigClass=&/F_MED_Dino_Collector_Deform_CtrlRig_sculpters.F_MED_Dino_Collector_Deform_CtrlRig_sculpters_C}
   }
   n11 = "Constraint - dyn_hip_hanging_01_MSTR (PP)"!(ComponentPose=Local To Component(LocalPose=@n10), Alpha=1, ConstraintWeights_0=0.3938, ConstraintWeights_1:float=0.6062) -> (Pose) {
    alphaBoolBlend {blendOption=Linear}
    boneToModify {boneName=dyn_hip_hanging_01_MSTR}
    constraintSetup [{targetBone {boneName=pelvis}}, {targetBone {boneName=thigh_twist_01_l}}]
    constraintWeights [1,1]
   }
   n12 = "Constraint - dyn_hip_hanging_02_MSTR (PP)"!(ComponentPose=@n11, Alpha=1, ConstraintWeights_0=0.3882, ConstraintWeights_1:float=0.6118) -> (Pose) {
    alphaBoolBlend {blendOption=Linear}
    boneToModify {boneName=dyn_hip_hanging_02_MSTR}
    constraintSetup [{targetBone {boneName=pelvis}}, {targetBone {boneName=thigh_twist_01_l}}]
    constraintWeights [1,1]
   }
   n13 = "Constraint - dyn_hip_hanging_03_MSTR (PP)"!(ComponentPose=@n12, Alpha=1, ConstraintWeights_0=0.7854, ConstraintWeights_1:float=0.2146) -> (Pose) {
    alphaBoolBlend {blendOption=Linear}
    boneToModify {boneName=dyn_hip_hanging_03_MSTR}
    constraintSetup [{targetBone {boneName=pelvis}}, {targetBone {boneName=thigh_twist_01_l}}]
    constraintWeights [1,1]
   }
   n14 = "Constraint - dyn_hip_hanging_04_MSTR (PP)"!(ComponentPose=@n13, Alpha=1, ConstraintWeights_0=0.9656, ConstraintWeights_1:float=0.0344) -> (Pose) {
    alphaBoolBlend {blendOption=Linear}
    boneToModify {boneName=dyn_hip_hanging_04_MSTR}
    constraintSetup [{targetBone {boneName=pelvis}}, {targetBone {boneName=thigh_twist_01_l}}]
    constraintWeights [1,1]
   }
   n15 = "Constraint - dyn_bones_01_MSTR (PPPPP)"!(
    ComponentPose=@n14
    Alpha=1
    ConstraintWeights_0=0.1947
    ConstraintWeights_1:float=0.3287
    ConstraintWeights_2:float=0.151
    ConstraintWeights_3:float=0.2622
    ConstraintWeights_4:float=0.0634
   ) -> (Pose) {
    alphaBoolBlend {blendOption=Linear}
    boneToModify {boneName=dyn_bones_01_MSTR}
    constraintSetup [
     {targetBone {boneName=spine_01}}
     {targetBone {boneName=spine_02}}
     {targetBone {boneName=spine_03}}
     {targetBone {boneName=spine_04}}
     {targetBone {boneName=pelvis}}
    ]
    constraintWeights [1,1,1,1,1]
   }
   n16 = "Constraint - dyn_bones_02_MSTR (PPPPP)"!(
    ComponentPose=@n15
    Alpha=1
    ConstraintWeights_0=0.2196
    ConstraintWeights_1:float=0.3713
    ConstraintWeights_2:float=0.099
    ConstraintWeights_3:float=0.2049
    ConstraintWeights_4:float=0.1052
   ) -> (Pose) {
    alphaBoolBlend {blendOption=Linear}
    boneToModify {boneName=dyn_bones_02_MSTR}
    constraintSetup [
     {targetBone {boneName=spine_01}}
     {targetBone {boneName=spine_02}}
     {targetBone {boneName=spine_03}}
     {targetBone {boneName=spine_04}}
     {targetBone {boneName=pelvis}}
    ]
    constraintWeights [1,1,1,1,1]
   }
   n17 = "Constraint - dyn_bones_03_MSTR (PPPPP)"!(
    ComponentPose=@n16
    Alpha=1
    ConstraintWeights_0=0.2364
    ConstraintWeights_1:float=0.3443
    ConstraintWeights_2:float=0.0784
    ConstraintWeights_3:float=0.1812
    ConstraintWeights_4:float=0.1596
   ) -> (Pose) {
    alphaBoolBlend {blendOption=Linear}
    boneToModify {boneName=dyn_bones_03_MSTR}
    constraintSetup [
     {targetBone {boneName=spine_01}}
     {targetBone {boneName=spine_02}}
     {targetBone {boneName=spine_03}}
     {targetBone {boneName=spine_04}}
     {targetBone {boneName=pelvis}}
    ]
    constraintWeights [1,1,1,1,1]
   }
   n18 = "Constraint - dyn_shoulder_pad_MSTR (PP)"!(ComponentPose=@n17, Alpha=1, ConstraintWeights_0=0.5, ConstraintWeights_1:float=0.5) -> (Pose) {
    alphaBoolBlend {blendOption=Linear}
    boneToModify {boneName=dyn_shoulder_pad_MSTR}
    constraintSetup [{targetBone {boneName=upperarm_l}}, {targetBone {boneName=upperarm_twist_01_l}}]
    constraintWeights [1,1]
   }
   n19 = Get thigh_rot_up_l() -> (thigh_rot_up_l:double)
   n20 = "Transform (Modify) Bone - Bone: dyn_sword_MSTR"(ComponentPose=@n18, Alpha=@n19, Translation="-0.725,-4.349,-0.835", Rotation="7.625,0.153,-24.921", Scale="1,1,1") {
    rotationMode=BMM_Additive
    rotationSpace=BCS_BoneSpace
    translationMode=BMM_Additive
    translationSpace=BCS_BoneSpace
    alphaBoolBlend {blendOption=Linear}
    boneToModify {boneName=dyn_sword_MSTR}
    rotation=<rot 7.625 0.153 -24.9208>
    translation=<-0.7253 -4.349 -0.835>
   }
   n21 = "Transform (Modify) Bone - Bone: dyn_hip_hanging_01_MSTR"(ComponentPose=@n20, Alpha=@n19, Translation="3.200,-2.131,7.095", Rotation="33,-13.950,-22", Scale="1,1,1") {
    rotationMode=BMM_Additive
    rotationSpace=BCS_BoneSpace
    translationMode=BMM_Additive
    translationSpace=BCS_BoneSpace
    alphaBoolBlend {blendOption=Linear}
    boneToModify {boneName=dyn_hip_hanging_01_MSTR}
    rotation=<rot 33 -13.95 -22>
    translation=<3.2 -2.131 7.0949>
   }
   n22 = "Transform (Modify) Bone - Bone: dyn_hip_hanging_02_MSTR"(ComponentPose=@n21, Alpha=@n19, Translation="-0.062,1.623,1.892", Rotation="16.776,5.668,-12.927", Scale="1,1,1") {
    rotationMode=BMM_Additive
    rotationSpace=BCS_BoneSpace
    translationMode=BMM_Additive
    translationSpace=BCS_BoneSpace
    alphaBoolBlend {blendOption=Linear}
    boneToModify {boneName=dyn_hip_hanging_02_MSTR}
    rotation=<rot 16.776 5.668 -12.9272>
    translation=<-0.062 1.623 1.892>
   }
   n23 = "Transform (Modify) Bone - Bone: dyn_belt_l_MSTR"(ComponentPose=@n22, Alpha=@n19, Translation="-5.656,0.077,-1.547", Rotation="0.000,-10.000,5.000", Scale="1,1,1") {
    rotationMode=BMM_Additive
    rotationSpace=BCS_BoneSpace
    translationMode=BMM_Additive
    translationSpace=BCS_BoneSpace
    alphaBoolBlend {blendOption=Linear}
    boneToModify {boneName=dyn_belt_l_MSTR}
    rotation {roll=5, yaw=-10.0003}
    translation=<-5.656 0.077 -1.547>
   }
   n24 = "Transform (Modify) Bone - Bone: dyn_hip_rope_MSTR"(ComponentPose=@n23, Alpha=@n19, Translation="-7.000,4.126,1.800", Rotation="25,0,0.000", Scale="1,1,1") {
    rotationMode=BMM_Additive
    rotationSpace=BCS_BoneSpace
    translationMode=BMM_Additive
    translationSpace=BCS_BoneSpace
    alphaBoolBlend {blendOption=Linear}
    boneToModify {boneName=dyn_hip_rope_MSTR}
    rotation {pitch=25}
    translation=<-7 4.126 1.8>
   }
   n25 = "Transform (Modify) Bone - Bone: dyn_glute_l"(ComponentPose=@n24, Alpha=Get thigh_rot_up_l(), Translation="-12,2,0", Rotation="0,0,0", Scale="1,1,1") {
    translationMode=BMM_Additive
    translationSpace=BCS_BoneSpace
    alphaBoolBlend {blendOption=Linear}
    boneToModify {boneName=dyn_glute_l}
    translation=<-12 2>
   }
   n26 = "Transform (Modify) Bone - Bone: dyn_glute_r"(ComponentPose=@n25, Alpha=Get thigh_rot_up_r() -> (thigh_rot_up_r:double), Translation="-12,1,0", Rotation="0,0,0", Scale="1,1,1") {
    translationMode=BMM_Additive
    translationSpace=BCS_BoneSpace
    alphaBoolBlend {blendOption=Linear}
    boneToModify {boneName=dyn_glute_r}
    translation=<-12 1>
   }
   n27 = "Transform (Modify) Bone - Bone: dyn_knee_pad_r"(ComponentPose=@n26, Alpha=Get knee_rot_r() -> (knee_rot_r:double), Translation="6,1.050,-0.000", Rotation="0.000,-30,0.000", Scale="1,1,1") {
    rotationMode=BMM_Additive
    rotationSpace=BCS_BoneSpace
    translationMode=BMM_Additive
    translationSpace=BCS_BoneSpace
    alphaBoolBlend {blendOption=Linear}
    boneToModify {boneName=dyn_knee_pad_r}
    rotation {yaw=-30}
    translation=<6 1.05>
   }
   n28 = Anim Dynamics - Bone: dyn_leg_bones_01(ComponentPose=@n27, Alpha=1, ExternalForce="0,0,0") {
    angularDampingOverride=0.995
    bGravityOverrideInSimSpace=true
    bOverrideAngularDamping=true
    bUseGravityOverride=true
    simulationSpace=RootRelative
    alphaBoolBlend {blendOption=Linear}
    boundBone {boneName=dyn_leg_bones_01}
    componentAppliedLinearAccClamp=<3000 3000 3000>
    componentLinearAccScale=<0.3 0.3 0.3>
    componentLinearVelScale=<0.3 0.3 0.3>
    gravityOverride {z=-980}
    physicsBodyDefinitions [{
      sphereCollisionRadius=0
      boundBone {boneName=dyn_leg_bones_01}
      boxExtents=<10 5 5>
      constraintSetup {angularLimitsMax=<15 0 45>, angularLimitsMin=<-15 -45 -45>, angularTarget=<0 0 0>}
      localJointOffset=<2 0 0>
     }]
    relativeSpaceBone {boneName=calf_r}
   }
   n29 = Anim Dynamics - Bone: dyn_leg_bones_02(ComponentPose=@n28, Alpha=1, ExternalForce="0,0,0") {
    angularDampingOverride=0.999
    bGravityOverrideInSimSpace=true
    bOverrideAngularDamping=true
    bUseGravityOverride=true
    simulationSpace=RootRelative
    alphaBoolBlend {blendOption=Linear}
    boundBone {boneName=dyn_leg_bones_02}
    componentAppliedLinearAccClamp=<3000 3000 3000>
    componentLinearAccScale=<0.3 0.3 0.3>
    componentLinearVelScale=<0.3 0.3 0.3>
    gravityOverride {z=-980}
    physicsBodyDefinitions [{
      sphereCollisionRadius=0
      boundBone {boneName=dyn_leg_bones_02}
      boxExtents=<10 5 5>
      constraintSetup {angularLimitsMax=<15 0 45>, angularLimitsMin=<-15 -45 -45>, angularTarget=<0 0 0>}
      localJointOffset=<1.5 0 0>
     }]
    relativeSpaceBone {boneName=calf_r}
   }
   n30 = Anim Dynamics - Bone: dyn_leg_bones_03(ComponentPose=@n29, Alpha=1, ExternalForce="0,0,0") {
    angularDampingOverride=0.995
    bGravityOverrideInSimSpace=true
    bOverrideAngularDamping=true
    bUseGravityOverride=true
    simulationSpace=RootRelative
    alphaBoolBlend {blendOption=Linear}
    boundBone {boneName=dyn_leg_bones_03}
    componentAppliedLinearAccClamp=<3000 3000 3000>
    componentLinearAccScale=<0.3 0.3 0.3>
    componentLinearVelScale=<0.3 0.3 0.3>
    gravityOverride {z=-980}
    physicsBodyDefinitions [{
      sphereCollisionRadius=0
      boundBone {boneName=dyn_leg_bones_03}
      boxExtents=<10 5 5>
      constraintSetup {angularLimitsMax=<15 0 45>, angularLimitsMin=<-15 -45 -45>, angularTarget=<0 0 0>}
      localJointOffset=<3 0 0>
     }]
    relativeSpaceBone {boneName=calf_r}
   }
   n31 = Anim Dynamics - Bone: dyn_leg_bones_04(ComponentPose=@n30, Alpha=1, ExternalForce="0,0,0") {
    angularDampingOverride=0.998
    bGravityOverrideInSimSpace=true
    bOverrideAngularDamping=true
    bUseGravityOverride=true
    simulationSpace=RootRelative
    alphaBoolBlend {blendOption=Linear}
    boundBone {boneName=dyn_leg_bones_04}
    componentAppliedLinearAccClamp=<3000 3000 3000>
    componentLinearAccScale=<0.3 0.3 0.3>
    componentLinearVelScale=<0.3 0.3 0.3>
    gravityOverride {z=-980}
    physicsBodyDefinitions [{
      sphereCollisionRadius=0
      boundBone {boneName=dyn_leg_bones_04}
      boxExtents=<10 5 5>
      constraintSetup {angularLimitsMax=<15 0 45>, angularLimitsMin=<-15 -45 -45>, angularTarget=<0 0 0>}
      localJointOffset=<2 0 0>
     }]
    relativeSpaceBone {boneName=calf_r}
   }
   n32 = Anim Dynamics - Bone: dyn_leg_bones_05(ComponentPose=@n31, Alpha=1, ExternalForce="0,0,0") {
    angularDampingOverride=0.998
    bGravityOverrideInSimSpace=true
    bOverrideAngularDamping=true
    bUseGravityOverride=true
    simulationSpace=RootRelative
    alphaBoolBlend {blendOption=Linear}
    boundBone {boneName=dyn_leg_bones_05}
    componentAppliedLinearAccClamp=<3000 3000 3000>
    componentLinearAccScale=<0.3 0.3 0.3>
    componentLinearVelScale=<0.3 0.3 0.3>
    gravityOverride {z=-980}
    physicsBodyDefinitions [{
      sphereCollisionRadius=0
      boundBone {boneName=dyn_leg_bones_05}
      boxExtents=<10 5 5>
      constraintSetup {angularLimitsMax=<15 0 45>, angularLimitsMin=<-15 -45 -45>, angularTarget=<0 0 0>}
      localJointOffset=<2.5 0 0>
     }]
    relativeSpaceBone {boneName=calf_r}
   }
   n33 = Anim Dynamics - Bone: dyn_leg_bones_06(ComponentPose=@n32, Alpha=1, ExternalForce="0,0,0") {
    angularDampingOverride=0.99
    bGravityOverrideInSimSpace=true
    bOverrideAngularDamping=true
    bUseGravityOverride=true
    simulationSpace=RootRelative
    alphaBoolBlend {blendOption=Linear}
    boundBone {boneName=dyn_leg_bones_06}
    componentAppliedLinearAccClamp=<3000 3000 3000>
    componentLinearAccScale=<0.3 0.3 0.3>
    componentLinearVelScale=<0.3 0.3 0.3>
    gravityOverride {z=-980}
    physicsBodyDefinitions [{
      sphereCollisionRadius=0
      boundBone {boneName=dyn_leg_bones_06}
      boxExtents=<10 5 5>
      constraintSetup {angularLimitsMax=<15 0 45>, angularLimitsMin=<-15 -45 -45>, angularTarget=<0 0 0>}
      localJointOffset=<2 0 0>
     }]
    relativeSpaceBone {boneName=calf_r}
   }
   n34 = Anim Dynamics - Bone: dyn_bones_01(ComponentPose=@n33, Alpha=1, ExternalForce="0,0,0") {
    angularDampingOverride=0.999
    bGravityOverrideInSimSpace=true
    bOverrideAngularDamping=true
    bUseGravityOverride=true
    alphaBoolBlend {blendOption=Linear}
    boundBone {boneName=dyn_bones_01}
    componentAppliedLinearAccClamp=<3000 3000 3000>
    componentLinearAccScale=<0.3 0.3 0.3>
    componentLinearVelScale=<0.3 0.3 0.3>
    gravityOverride {z=-980}
    physicsBodyDefinitions [{
      sphereCollisionRadius=0
      boundBone {boneName=dyn_bones_01}
      boxExtents=<10 8 8>
      constraintSetup {angularLimitsMax=<15 0 45>, angularLimitsMin=<-15 -45 -45>, angularTarget=<0 0 0>}
      localJointOffset=<1.5 0 0>
     }]
   }
   n35 = Anim Dynamics - Bone: dyn_bones_02(ComponentPose=@n34, Alpha=1, ExternalForce="0,0,0") {
    angularDampingOverride=0.999
    bGravityOverrideInSimSpace=true
    bOverrideAngularDamping=true
    bUseGravityOverride=true
    alphaBoolBlend {blendOption=Linear}
    boundBone {boneName=dyn_bones_02}
    componentAppliedLinearAccClamp=<3000 3000 3000>
    componentLinearAccScale=<0.3 0.3 0.3>
    componentLinearVelScale=<0.3 0.3 0.3>
    gravityOverride {z=-980}
    physicsBodyDefinitions [{
      sphereCollisionRadius=0
      boundBone {boneName=dyn_bones_02}
      boxExtents=<10 8 8>
      constraintSetup {angularLimitsMax=<15 0 45>, angularLimitsMin=<-15 -45 -45>, angularTarget=<0 0 0>}
      localJointOffset=<1 0 0>
     }]
   }
   n36 = Anim Dynamics - Bone: dyn_bones_03(ComponentPose=@n35, Alpha=1, ExternalForce="0,0,0") {
    angularDampingOverride=0.999
    bGravityOverrideInSimSpace=true
    bOverrideAngularDamping=true
    bUseGravityOverride=true
    alphaBoolBlend {blendOption=Linear}
    boundBone {boneName=dyn_bones_03}
    componentAppliedLinearAccClamp=<3000 3000 3000>
    componentLinearAccScale=<0.3 0.3 0.3>
    componentLinearVelScale=<0.3 0.3 0.3>
    gravityOverride {z=-980}
    physicsBodyDefinitions [{
      sphereCollisionRadius=0
      boundBone {boneName=dyn_bones_03}
      boxExtents=<8 10 10>
      constraintSetup {angularLimitsMax=<15 0 45>, angularLimitsMin=<-15 -45 -45>, angularTarget=<0 0 0>}
      localJointOffset=<1 0 0>
     }]
   }
   n37 = Get thigh_rot_up_l()
   n38 = "Transform (Modify) Bone - Bone: dyn_bones_01_MSTR"(ComponentPose=@n36, Alpha=@n37, Translation="0,0,0", Rotation="10.000,0.000,0.000", Scale="1,1,1") {
    rotationMode=BMM_Additive
    rotationSpace=BCS_BoneSpace
    alphaBoolBlend {blendOption=Linear}
    boneToModify {boneName=dyn_bones_01_MSTR}
    rotation {pitch=10}
   }
   n39 = "Transform (Modify) Bone - Bone: dyn_bones_02_MSTR"(ComponentPose=@n38, Alpha=@n37, Translation="0,0,0", Rotation="9.055,25.338,4.261", Scale="1,1,1") {
    rotationMode=BMM_Additive
    rotationSpace=BCS_BoneSpace
    alphaBoolBlend {blendOption=Linear}
    boneToModify {boneName=dyn_bones_02_MSTR}
    rotation=<rot 9.055 25.3377 4.261>
   }
   n40 = "Transform (Modify) Bone - Bone: dyn_bones_03_MSTR"(ComponentPose=@n39, Alpha=@n37, Translation="0,0,0", Rotation="19.291,-15.916,-5.382", Scale="1,1,1") {
    rotationMode=BMM_Additive
    rotationSpace=BCS_BoneSpace
    alphaBoolBlend {blendOption=Linear}
    boneToModify {boneName=dyn_bones_03_MSTR}
    rotation=<rot 19.2909 -15.9164 -5.3821>
   }
   n41 = Anim Dynamics - Bone: dyn_hip_hanging_04(ComponentPose=@n40, Alpha=1, ExternalForce="0,0,0") {
    angularDampingOverride=0.999
    bGravityOverrideInSimSpace=true
    bOverrideAngularDamping=true
    bUseGravityOverride=true
    alphaBoolBlend {blendOption=Linear}
    boundBone {boneName=dyn_hip_hanging_04}
    componentAppliedLinearAccClamp=<3000 3000 3000>
    componentLinearAccScale=<0.3 0.3 0.3>
    componentLinearVelScale=<0.3 0.3 0.3>
    gravityOverride {z=-980}
    physicsBodyDefinitions [{
      sphereCollisionRadius=0
      boundBone {boneName=dyn_hip_hanging_04}
      boxExtents=<10 15 10>
      constraintSetup {angularLimitsMax=<0 5 0>, angularLimitsMin=<0 -45 0>, angularTarget=<0 0 0>}
      localJointOffset=<3 0 0>
     }]
   }
   n42 = Anim Dynamics - Bone: dyn_hip_hanging_03(ComponentPose=@n41, Alpha=1, ExternalForce="0,0,0") {
    angularDampingOverride=0.999
    bGravityOverrideInSimSpace=true
    bOverrideAngularDamping=true
    bUseGravityOverride=true
    alphaBoolBlend {blendOption=Linear}
    boundBone {boneName=dyn_hip_hanging_03}
    componentAppliedLinearAccClamp=<3000 3000 3000>
    componentLinearAccScale=<0.3 0.3 0.3>
    componentLinearVelScale=<0.3 0.3 0.3>
    gravityOverride {z=-980}
    physicsBodyDefinitions [{
      sphereCollisionRadius=0
      boundBone {boneName=dyn_hip_hanging_03}
      boxExtents=<15 15 10>
      constraintSetup {angularLimitsMax=<0 0 0>, angularLimitsMin=<0 -20 0>, angularTarget=<0 0 0>}
      localJointOffset=<1 0 0>
     }]
   }
   n43 = Anim Dynamics - Bone: dyn_sword_bones_01(ComponentPose=@n42, Alpha=1, ExternalForce="0,0,0") {
    angularDampingOverride=0.999
    bGravityOverrideInSimSpace=true
    bOverrideAngularDamping=true
    bUseGravityOverride=true
    simulationSpace=RootRelative
    alphaBoolBlend {blendOption=Linear}
    boundBone {boneName=dyn_sword_bones_01}
    componentAppliedLinearAccClamp=<3000 3000 3000>
    componentLinearAccScale=<0.3 0.3 0.3>
    componentLinearVelScale=<0.3 0.3 0.3>
    gravityOverride {z=-980}
    physicsBodyDefinitions [{
      sphereCollisionRadius=0
      boundBone {boneName=dyn_sword_bones_01}
      boxExtents=<10 6 6>
      constraintSetup {angularLimitsMax=<15 28 45>, angularLimitsMin=<-15 -45 -45>, angularTarget=<0 0 0>}
      localJointOffset=<1 0 0>
     }]
    relativeSpaceBone {boneName=dyn_sword_MSTR}
   }
   n44 = Anim Dynamics - Bone: dyn_sword_bones_02(ComponentPose=@n43, Alpha=1, ExternalForce="0,0,0") {
    angularDampingOverride=0.999
    bGravityOverrideInSimSpace=true
    bOverrideAngularDamping=true
    bUseGravityOverride=true
    simulationSpace=RootRelative
    alphaBoolBlend {blendOption=Linear}
    boundBone {boneName=dyn_sword_bones_02}
    componentAppliedLinearAccClamp=<3000 3000 3000>
    componentLinearAccScale=<0.3 0.3 0.3>
    componentLinearVelScale=<0.3 0.3 0.3>
    gravityOverride {z=-980}
    physicsBodyDefinitions [{
      sphereCollisionRadius=0
      boundBone {boneName=dyn_sword_bones_02}
      boxExtents=<10 6 6>
      constraintSetup {angularLimitsMax=<15 0 45>, angularLimitsMin=<-15 -45 -45>, angularTarget=<0 0 0>}
      localJointOffset=<2 0 0>
     }]
    relativeSpaceBone {boneName=dyn_sword_MSTR}
   }
   n45 = Anim Dynamics - Bone: dyn_sword_bones_03(ComponentPose=@n44, Alpha=1, ExternalForce="0,0,0") {
    angularDampingOverride=0.999
    bGravityOverrideInSimSpace=true
    bOverrideAngularDamping=true
    bUseGravityOverride=true
    simulationSpace=RootRelative
    alphaBoolBlend {blendOption=Linear}
    boundBone {boneName=dyn_sword_bones_03}
    componentAppliedLinearAccClamp=<3000 3000 3000>
    componentLinearAccScale=<0.3 0.3 0.3>
    componentLinearVelScale=<0.3 0.3 0.3>
    gravityOverride {z=-980}
    physicsBodyDefinitions [{
      sphereCollisionRadius=0
      boundBone {boneName=dyn_sword_bones_03}
      boxExtents=<12 6 6>
      constraintSetup {angularLimitsMax=<15 45 45>, angularLimitsMin=<-15 -45 -45>, angularTarget=<0 0 0>}
      localJointOffset=<3 0 0>
     }]
    relativeSpaceBone {boneName=dyn_sword_MSTR}
   }
   n46 = Anim Dynamics - Bone: dyn_sword_bones_04(ComponentPose=@n45, Alpha=1, ExternalForce="0,0,0") {
    angularDampingOverride=0.999
    bGravityOverrideInSimSpace=true
    bOverrideAngularDamping=true
    bUseGravityOverride=true
    simulationSpace=RootRelative
    alphaBoolBlend {blendOption=Linear}
    boundBone {boneName=dyn_sword_bones_04}
    componentAppliedLinearAccClamp=<3000 3000 3000>
    componentLinearAccScale=<0.3 0.3 0.3>
    componentLinearVelScale=<0.3 0.3 0.3>
    gravityOverride {z=-980}
    physicsBodyDefinitions [{
      sphereCollisionRadius=0
      boundBone {boneName=dyn_sword_bones_04}
      boxExtents=<12 6 6>
      constraintSetup {angularLimitsMax=<15 45 45>, angularLimitsMin=<-15 -45 -45>, angularTarget=<0 0 0>}
      localJointOffset=<2 0 0>
     }]
    relativeSpaceBone {boneName=dyn_sword_MSTR}
   }
   Save cached pose 'prePose'(Pose=Component To Local(ComponentPose=@n46)) {cacheName=prePose}
   n47 = RigidBody!(ComponentPose=Local To Component(LocalPose=Use cached pose 'prePose'()), Alpha, OverrideWorldGravity:Vector="0,0,-980", ExternalForce) -> (Pose) {
    bClampLinearTranslationLimitToRefPose=true
    bOverrideWorldGravity=true
    overridePhysicsAsset=&/F_MED_Dino_Collector_dyn_sword
    simulationSpace=BaseBoneSpace
    alphaBoolBlend {blendOption=Linear}
    baseBoneRef {boneName=spine_01}
    componentAppliedLinearAccClamp=<650 1000 650>
    componentLinearAccScale=<0.8 0.4 0.5>
    componentLinearVelScale=<0.5 0.4 0.5>
    overrideWorldGravity {z=-980}
   }
   n48 = "Transform (Modify) Bone - Bone: dyn_sword_01"(ComponentPose=@n47, Alpha=Get thigh_rot_up_l(), Translation="0,0,0", Rotation="0.000,0.000,-10", Scale="1,1,1") {
    rotationMode=BMM_Additive
    rotationSpace=BCS_BoneSpace
    translationMode=BMM_Additive
    translationSpace=BCS_BoneSpace
    alphaBoolBlend {blendOption=Linear}
    boneToModify {boneName=dyn_sword_01}
    rotation {roll=-10}
   }
   n49 = "Transform (Modify) Bone - Bone: dyn_sword_02"(ComponentPose=@n48, Alpha=Get isMovingSide() -> (isMovingSide:double), Translation="0,0,0", Rotation="-6.112,-0.088,18.998", Scale="1,1,1") {
    rotationMode=BMM_Additive
    rotationSpace=BCS_BoneSpace
    scaleMode=BMM_Additive
    scaleSpace=BCS_BoneSpace
    translationMode=BMM_Additive
    translationSpace=BCS_BoneSpace
    alphaBoolBlend {blendOption=Linear}
    alphaScaleBiasClamp {bInterpResult=true}
    boneToModify {boneName=dyn_sword_02}
    rotation=<rot -6.112 -0.088 18.998>
   }
   n50 = "Transform (Modify) Bone - Bone: dyn_sword_02"(ComponentPose=@n49, Alpha=Get isMoveingFwd() -> (isMoveingFwd:double), Translation="0,0,0", Rotation="1.047,-8.535,19.226", Scale="1,1,1") {
    rotationMode=BMM_Additive
    rotationSpace=BCS_BoneSpace
    scaleMode=BMM_Additive
    scaleSpace=BCS_BoneSpace
    translationMode=BMM_Additive
    translationSpace=BCS_BoneSpace
    alphaBoolBlend {blendOption=Linear}
    alphaScaleBiasClamp {bInterpResult=true}
    boneToModify {boneName=dyn_sword_02}
    rotation=<rot 1.047 -8.535 19.226>
   }
   n51 = Component To Local(ComponentPose=@n50)
   n52 = RigidBody(ComponentPose=Local To Component(LocalPose=Use cached pose 'prePose'()), OverrideWorldGravity="0,0,0") {
    bClampLinearTranslationLimitToRefPose=true
    bOverrideWorldGravity=true
    overridePhysicsAsset=&/F_MED_Dino_Collector_dyn_sword_diving
    simulationSpace=BaseBoneSpace
    alphaBoolBlend {blendOption=Linear}
    baseBoneRef {boneName=pelvis}
    componentAppliedLinearAccClamp=<650 1000 650>
    componentLinearAccScale=<0.1 0.1 0.1>
    componentLinearVelScale=<0.1 0.1 0.1>
    overrideWorldGravity {z=-980}
   }
   n53 = "Transform (Modify) Bone - Bone: dyn_sword_02"(ComponentPose=@n52, Alpha=1, Translation="0,0,0", Rotation="10,40,0", Scale="1,1,1") {
    rotationMode=BMM_Additive
    rotationSpace=BCS_BoneSpace
    scaleMode=BMM_Additive
    scaleSpace=BCS_BoneSpace
    translationMode=BMM_Additive
    translationSpace=BCS_BoneSpace
    alphaBoolBlend {blendOption=Linear}
    alphaScaleBiasClamp {bInterpResult=true}
    boneToModify {boneName=dyn_sword_02}
    rotation {pitch=10, yaw=40}
   }
   n54 = Component To Local(ComponentPose=@n53)
   n55 = Get bIsSkyDiving()
   n56 = RigidBody(
    ComponentPose=Local To Component(LocalPose=Blend Poses by bool(BlendPose_0=@n54, BlendPose_1=@n51, bActiveValue=@n55) {blendType=Linear})
    OverrideWorldGravity="0,0,-980"
   ) {
    bClampLinearTranslationLimitToRefPose=true
    bOverrideWorldGravity=true
    overridePhysicsAsset=&/F_MED_Dino_Collector_dyn_necklace
    simulationSpace=BaseBoneSpace
    alphaBoolBlend {blendOption=Linear}
    baseBoneRef {boneName=spine_05}
    componentAppliedLinearAccClamp=<2000 2000 1000>
    componentLinearAccScale {x=0.4}
    componentLinearVelScale {x=0.4}
    overrideWorldGravity {z=-980}
    simSpaceSettings {maxAngularAcceleration=100, maxAngularVelocity=5, maxLinearAcceleration=300, maxLinearVelocity=50, velocityScaleZ=0, worldAlpha=1}
   }
   Save cached pose 'rigidCache'(Pose=Component To Local(ComponentPose=@n56)) {cacheName=rigidCache}
   Get gravity_override() -> (gravity_override:Vector)
   Get gravity_override()
  ]}
 EventGraph {nodes [
   Event BlueprintUpdateAnimation()
   > n1 = Sequence!(execute) -> (then_0, then_1, then_2:Exec, then_3:Exec, then_4:Exec, then_5:Exec)
   n2 = BreakTransform(InTransform=GetDeltaTransformFromRefPose(self=GetOwnerSkeletalMeshComponent(), BoneName=thigh_r, BaseName=pelvis))
   > Set thigh_rot_up_r(thigh_rot_up_r:double=MapRangeClamped(Value=BreakRotator(InRot=@n2.Rotation).Yaw, InRangeA=-15, InRangeB=-100)) -> (Output_Get:double)
   n3 = BreakTransform(InTransform=GetDeltaTransformFromRefPose(self=GetOwnerSkeletalMeshComponent(), BoneName=thigh_l, BaseName=pelvis))
   Set thigh_rot_up_l(execute=@n1.then_1, thigh_rot_up_l:double=MapRangeClamped(Value=BreakRotator(InRot=@n3.Rotation).Yaw, InRangeA=-15, InRangeB=-100)) -> (Output_Get:double)
   n4 = Gravity Override State Function DEPRECATED(
    execute=@n1.then_3
    CharacterPartAnimInstance=Self-Reference()
    GameTimeInSeconds=GetGameTimeInSeconds()
    GravityOverrideParams=Get Gravity Override Params()
   )
   > Set gravity_override(gravity_override=@n4)
   n5 = PlayerMovement(execute=@n1.then_4, AnimInstance=Self-Reference(), XDotProductCoefficient=1, YDotProductCoefficient=1, ZDotProductCoefficient=1)
   > Set isMovingSide(isMovingSide:double=MapRangeClamped(Value="Absolute (Float)"(A=@n5), InRangeA=0, InRangeB=1)) -> (Output_Get:double)
   > Set isMoveingFwd(isMoveingFwd:double=MapRangeClamped(Value=@n5.YMovement, InRangeA=0, InRangeB=1)) -> (Output_Get:double)
   Set isSkyDiving(execute=@n1.then_5, isSkyDiving:double="To Float (Boolean)"(InBool=Get bIsSkydiveDiveMode())) -> (Output_Get:double)
   n6 = BreakTransform(InTransform=GetDeltaTransformFromRefPose(self=GetOwnerSkeletalMeshComponent(), BoneName=calf_r, BaseName=thigh_r))
   Set knee_rot_r(execute=@n1.then_2, knee_rot_r:double=MapRangeClamped(Value=BreakRotator(InRot=@n6.Rotation).Yaw, InRangeA=10, InRangeB=110)) -> (Output_Get:double)
   Event BlueprintInitializeAnimation()
   > Set Skel_Mesh(Skel_Mesh=GetOwnerSkeletalMeshComponent())
  ]}
]
variables [
 Gravity Override Params {
  type=GravityOverrideParamsStruct
  default {
   skydive_DivingDown {gravityOverride=<-500 0 0>, windFrequency=<1.5 6 4>}
   skydive_FloatingDown {gravityOverride=<0 -100 -500>, windAmplitude=<200 -100 50>, windOffset=<0 -200 50>}
   skydive_GoingUp {gravityOverride=<0 -300 -100>, windAmplitude=<200 -500 100>}
   skydive_Parachuting {gravityOverride=<0 -500 -980>, windAmplitude=<200 -300 400>}
  }
 }
 gravity_override {type=Vector}
 isMoveingFwd {type=number}
 isMovingSide {type=number}
 isSkyDiving {type=number}
 knee_rot_r {type=number}
 shldr_l_alpha {type=number}
 Skel_Mesh {type=SkeletalMeshComponent}
 thigh_rot_up_l {type=number}
 thigh_rot_up_r {type=number}
]
