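# Physics asset for the dynamic hair of the F_MED_Dino_Collector female player body, covering back hair, temple braids, hair tips and a hanging bone ornament.
# Capsules on head, neck, upper spine, clavicles and upper arms form a body collision frame joined by free-swinging constraints.
# Hair hangs from the head as tightly limited capsule chains; the back chain starts at a free master bone and branches left and right partway down.
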PhysicsAsset /Game/Characters/Player/Female/Medium/Bodies/F_MED_Dino_Collector/Meshes/Parts/F_MED_Dino_Collector_dyn_hair
bodies {
 clavicle_l [{shapeType=Capsule, length=9.3737, radius=5.574, center=<4.0899 -0.4322 -0.7731>, rotation=<rot -0.4769 -84.5415 -85.0209>}]
 clavicle_r [{shapeType=Capsule, length=9.3737, radius=5.574, center=<-4.0899 0.4322 0.7731>, rotation=<rot -0.4769 -84.5415 94.9791>}]
 dyn_bone_hanging [{shapeType=Capsule, length=2.9503, radius=0.9928, center=<2.2108 -0.2712 0>, rotation=<rot 84.9871 180 180>}]
 dyn_bone_hanging_base [{shapeType=Capsule, length=2.8253, radius=1.4928, center=<2.9418 0 -0.0922>, rotation=<rot 90 0.0396 0.0397>}]
 dyn_hairTips_01 [{shapeType=Capsule, length=4.0007, radius=2.2378, center=<4 0 0>, rotation=<rot 90 0 0>}]
 dyn_hairTips_02 [{shapeType=Capsule, length=2.6257, radius=2.2378, center=<3 0 0>, rotation=<rot 90 0 0>}]
 dyn_hairTips_03 [{shapeType=Capsule, length=4.0007, radius=2.2378, center=<3 0 0>, rotation=<rot 90 0 0>}]
 dyn_hairTips_04 [{shapeType=Capsule, length=3.2507, radius=1.7378, center=<3 0 0>, rotation=<rot 90 0 0>}]
 dyn_hairTips_05 [{shapeType=Capsule, length=4.0007, radius=2.2378, center=<4 0 0>, rotation=<rot 90 0 0>}]
 dyn_hair_bk_01 [{shapeType=Capsule, length=5.8064, radius=3.4309, center=<2 0 2.6>, rotation=<rot 0 0 90>}]
 dyn_hair_bk_02 [{shapeType=Capsule, length=6.0564, radius=3.1809, center=<2.6688 0 0>, rotation=<rot 0 0 90>}]
 dyn_hair_bk_03 [{shapeType=Capsule, length=3.7699, radius=2.8018, center=<4.6672 0 -0.294>, rotation=<rot 85 0 0>}]
 dyn_hair_bk_MSTR [{shapeType=Capsule, length=1, radius=1}]
 dyn_hair_bk_l_00 [{shapeType=Capsule, length=2.75, radius=1.25, center=<3 0 0>, rotation=<rot 90 0 0>}]
 dyn_hair_bk_l_01 [{shapeType=Capsule, length=2.5, radius=1.5, center=<3 0 0>, rotation=<rot 90 0 0>}]
 dyn_hair_bk_l_02 [{shapeType=Capsule, length=2.75, radius=2, center=<3 0 0>, rotation=<rot 90 0 0>}]
 dyn_hair_bk_r_00 [{shapeType=Capsule, length=2.75, radius=1.25, center=<3 0 0>, rotation=<rot 90 0 0>}]
 dyn_hair_bk_r_01 [{shapeType=Capsule, length=2.5, radius=1.5, center=<3 0 0>, rotation=<rot 90 0 0>}]
 dyn_hair_bk_r_02 [{shapeType=Capsule, length=2.75, radius=2, center=<3 0 0>, rotation=<rot 90 0 0>}]
 dyn_templeBraid_01_l [{shapeType=Capsule, length=4, radius=1.25, center=<5.6664 0 0>, rotation=<rot 90 0 0>}]
 dyn_templeBraid_01_r [{shapeType=Capsule, length=5, radius=1.25, center=<5.5976 0 0.0002>, rotation=<rot 90 0 0>}]
 dyn_templeBraid_02_l [{shapeType=Capsule, length=2.875, radius=1.125, center=<3 0 0>, rotation=<rot 90 0 0>}]
 dyn_templeBraid_02_r [{shapeType=Capsule, length=4.625, radius=1.125, center=<3 0 0>, rotation=<rot 90 0 0>}]
 dyn_templeBraid_03_l [{shapeType=Capsule, length=2.375, radius=1, center=<3 0 0>, rotation=<rot 90 0 0>}]
 dyn_templeBraid_03_r [{shapeType=Capsule, length=3.375, radius=1.125, center=<3 0 0>, rotation=<rot 90 0 0>}]
 dyn_templeBraid_04_l [{shapeType=Capsule, length=2.375, radius=1.375, center=<3.465 0 0>, rotation=<rot 90 0 0>}]
 dyn_templeBraid_04_r [{shapeType=Capsule, length=2.375, radius=1.375, center=<3.3072 0 0>, rotation=<rot 90 0 0>}]
 dyn_templeBraid_Base_l [{shapeType=Capsule, length=1, radius=1, rotation=<rot 90 0 0>}]
 dyn_templeBraid_Base_r [{shapeType=Capsule, length=1, radius=1, rotation=<rot 90 0 0>}]
 head [
  {shapeType=Capsule, length=5.3905, radius=5.4204, center=<2.66 3.9667 -0.0047>, rotation=<rot -0.3094 99.2805 -89.484>}
  {shapeType=Capsule, length=3.5, radius=4.875, center=<8.4961 2.552 -2.8464>, rotation=<rot 0 20.0024 90>}
  {shapeType=Capsule, length=3.5, radius=4.875, center=<8.5582 2.5747 2.834>, rotation=<rot 0 20.0024 90>}
 ]
 neck_01 [{shapeType=Capsule, length=4.6405, radius=3.9204, center=<4.0596 0.6591 0>, rotation=<rot 90 14.0363 19.0362>}]
 spine_04 [
  {shapeType=Capsule, length=6.5, radius=10.175, center=<1.1042 3.6761 0.0004>}
  {shapeType=Capsule, length=4.5, radius=6.675, center=<5.7055 6.2751 0.0004>}
 ]
 spine_05 [{shapeType=Capsule, length=11.7812, radius=7.5, center=<2.1205 -0.408 0>}]
 upperarm_l [{shapeType=Capsule, length=8, radius=5.75, center=<-0.3062 3.4538 5.8142>, rotation=<rot 14.9999 0 -13.1779>}]
 upperarm_r [{shapeType=Capsule, length=18.5, radius=3.5, center=<-10.215 0.0001 0.4605>, rotation=<rot 90 0 0>}]
}
constraints [
 {
  bone1Name=head
  bone2Name=neck_01
  swing1LimitDegrees=45
  swing1Motion=Free
  swing2LimitDegrees=45
  swing2Motion=Free
  twistLimitDegrees=45
  twistMotion=Free
 }
 {bone1Name=dyn_templeBraid_01_l, bone2Name=dyn_templeBraid_Base_l, swing1LimitDegrees=60, swing2LimitDegrees=5, twistLimitDegrees=5}
 {bone1Name=dyn_templeBraid_02_l, bone2Name=dyn_templeBraid_01_l, swing1LimitDegrees=30, swing2LimitDegrees=15, twistLimitDegrees=5}
 {bone1Name=dyn_templeBraid_03_l, bone2Name=dyn_templeBraid_02_l, swing1LimitDegrees=30, swing2LimitDegrees=30, twistLimitDegrees=5}
 {bone1Name=dyn_templeBraid_04_l, bone2Name=dyn_templeBraid_03_l, swing1LimitDegrees=30, swing2LimitDegrees=30, twistLimitDegrees=5}
 {bone1Name=dyn_templeBraid_01_r, bone2Name=dyn_templeBraid_Base_r, swing1LimitDegrees=60, swing2LimitDegrees=5, twistLimitDegrees=5}
 {bone1Name=dyn_templeBraid_02_r, bone2Name=dyn_templeBraid_01_r, swing1LimitDegrees=30, swing2LimitDegrees=15, twistLimitDegrees=5}
 {bone1Name=dyn_templeBraid_03_r, bone2Name=dyn_templeBraid_02_r, swing1LimitDegrees=30, swing2LimitDegrees=30, twistLimitDegrees=5}
 {bone1Name=dyn_templeBraid_04_r, bone2Name=dyn_templeBraid_03_r, swing1LimitDegrees=30, swing2LimitDegrees=30, twistLimitDegrees=5}
 {
  bone1Name=spine_05
  bone2Name=spine_04
  swing1LimitDegrees=45
  swing1Motion=Free
  swing2LimitDegrees=45
  swing2Motion=Free
  twistLimitDegrees=45
  twistMotion=Free
 }
 {bone1Name=dyn_templeBraid_Base_r, bone2Name=head, swing1LimitDegrees=25, swing2LimitDegrees=30, twistLimitDegrees=5}
 {bone1Name=dyn_templeBraid_Base_l, bone2Name=head, swing1LimitDegrees=25, swing2LimitDegrees=30, twistLimitDegrees=5}
 {
  bone1Name=clavicle_r
  bone2Name=spine_05
  swing1LimitDegrees=45
  swing1Motion=Free
  swing2LimitDegrees=45
  swing2Motion=Free
  twistLimitDegrees=45
  twistMotion=Free
 }
 {
  bone1Name=upperarm_l
  bone2Name=clavicle_l
  swing1LimitDegrees=45
  swing1Motion=Free
  swing2LimitDegrees=45
  swing2Motion=Free
  twistLimitDegrees=45
  twistMotion=Free
 }
 {
  bone1Name=upperarm_r
  bone2Name=clavicle_r
  swing1LimitDegrees=45
  swing1Motion=Free
  swing2LimitDegrees=45
  swing2Motion=Free
  twistLimitDegrees=45
  twistMotion=Free
 }
 {
  bone1Name=neck_01
  bone2Name=spine_05
  swing1LimitDegrees=45
  swing1Motion=Free
  swing2LimitDegrees=45
  swing2Motion=Free
  twistLimitDegrees=45
  twistMotion=Free
 }
 {bone1Name=dyn_hair_bk_01, bone2Name=dyn_hair_bk_MSTR, swing1LimitDegrees=2, swing2LimitDegrees=20, twistLimitDegrees=2}
 {bone1Name=dyn_hair_bk_02, bone2Name=dyn_hair_bk_01, swing1LimitDegrees=2, swing2LimitDegrees=20, twistLimitDegrees=2}
 {bone1Name=dyn_hair_bk_03, bone2Name=dyn_hair_bk_02, swing1LimitDegrees=2, swing2LimitDegrees=45, twistLimitDegrees=2}
 {bone1Name=dyn_hair_bk_l_01, bone2Name=dyn_hair_bk_l_00, swing1LimitDegrees=2, swing2LimitDegrees=15, twistLimitDegrees=2}
 {bone1Name=dyn_hair_bk_l_02, bone2Name=dyn_hair_bk_l_01, swing1LimitDegrees=2, swing2LimitDegrees=15, twistLimitDegrees=2}
 {bone1Name=dyn_hair_bk_r_01, bone2Name=dyn_hair_bk_r_00, swing1LimitDegrees=2, swing2LimitDegrees=15, twistLimitDegrees=2}
 {bone1Name=dyn_hair_bk_r_02, bone2Name=dyn_hair_bk_r_01, swing1LimitDegrees=2, swing2LimitDegrees=15, twistLimitDegrees=2}
 {
  bone1Name=dyn_hair_bk_MSTR
  bone2Name=head
  swing1LimitDegrees=45
  swing1Motion=Free
  swing2LimitDegrees=45
  swing2Motion=Free
  twistLimitDegrees=45
  twistMotion=Free
 }
 {bone1Name=dyn_hairTips_01, bone2Name=head, swing1LimitDegrees=45, swing2LimitDegrees=35, twistLimitDegrees=2}
 {bone1Name=dyn_hairTips_02, bone2Name=head, swing1LimitDegrees=45, swing2LimitDegrees=30, twistLimitDegrees=2}
 {bone1Name=dyn_hairTips_03, bone2Name=head, swing1LimitDegrees=45, swing2LimitDegrees=45, twistLimitDegrees=2}
 {bone1Name=dyn_hairTips_04, bone2Name=head, swing1LimitDegrees=35, swing2LimitDegrees=35, twistLimitDegrees=2}
 {bone1Name=dyn_hairTips_05, bone2Name=head, swing1LimitDegrees=45, swing2LimitDegrees=35, twistLimitDegrees=2}
 {bone1Name=dyn_bone_hanging, bone2Name=dyn_bone_hanging_base, swing1LimitDegrees=10, swing2LimitDegrees=30, twistLimitDegrees=5}
 {bone1Name=dyn_bone_hanging_base, bone2Name=head, swing1LimitDegrees=10, swing2LimitDegrees=10, twistLimitDegrees=2}
 {bone1Name=dyn_hair_bk_l_00, bone2Name=dyn_hair_bk_02, swing1LimitDegrees=2, swing2LimitDegrees=15, twistLimitDegrees=2}
 {bone1Name=dyn_hair_bk_r_00, bone2Name=dyn_hair_bk_02, swing1LimitDegrees=2, swing2LimitDegrees=15, twistLimitDegrees=2}
 {
  bone1Name=clavicle_l
  bone2Name=spine_05
  swing1LimitDegrees=45
  swing1Motion=Free
  swing2LimitDegrees=45
  swing2Motion=Free
  twistLimitDegrees=45
  twistMotion=Free
 }
]
