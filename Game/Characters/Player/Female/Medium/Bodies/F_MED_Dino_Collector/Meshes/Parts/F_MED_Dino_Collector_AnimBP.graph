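# Post-process-style part AnimBP for the female Dino Collector that animates her hair and braids over copied body poses.
# A cached pose blends both copies at the upper arms, aiming back hair at spine_05 with a head-rotation-driven offset.
# bRigidBodyEnabled selects RigidBody physics, using gravity from a deprecated skydive-aware override function, or a fallback Look At chain.
# The hair-tip Anim Dynamics chain is unconnected to the output.
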AnimBlueprint /Game/Characters/Player/Female/Medium/Bodies/F_MED_Dino_Collector/Meshes/Parts/F_MED_Dino_Collector_AnimBP
target_skeleton=&/F_MED_Dino_Collector_Skeleton
graphs [
 AnimGraph {nodes [
   n1 = Anim Dynamics - Bone: dyn_hairTips_01(Alpha=1, ExternalForce="0,0,0") {
    angularDampingOverride=0.92
    angularSpringConstant=500
    bAngularSpring=true
    bGravityOverrideInSimSpace=true
    bOverrideAngularDamping=true
    bUseGravityOverride=true
    simulationSpace=RootRelative
    alphaBoolBlend {blendOption=Linear}
    boundBone {boneName=dyn_hairTips_01}
    componentAppliedLinearAccClamp=<500 500 500>
    componentLinearAccScale=<0.4 0.4 0.4>
    componentLinearVelScale=<0.35 0.35 0.35>
    gravityOverride {z=-980}
    physicsBodyDefinitions [{
      sphereCollisionRadius=0
      boundBone {boneName=dyn_hairTips_01}
      boxExtents=<15 15 6>
      constraintSetup {angularLimitsMax=<0 2 20>, angularLimitsMin=<0 -35 -20>, angularTarget=<0 0 0>}
      localJointOffset=<3 0 0>
     }]
    relativeSpaceBone {boneName=head}
   }
   n2 = Anim Dynamics - Bone: dyn_hairTips_02(ComponentPose=@n1, Alpha=1, ExternalForce="0,0,0") {
    angularDampingOverride=0.92
    angularSpringConstant=500
    bAngularSpring=true
    bGravityOverrideInSimSpace=true
    bOverrideAngularDamping=true
    bUseGravityOverride=true
    simulationSpace=RootRelative
    alphaBoolBlend {blendOption=Linear}
    boundBone {boneName=dyn_hairTips_02}
    componentAppliedLinearAccClamp=<500 500 500>
    componentLinearAccScale=<0.4 0.4 0.4>
    componentLinearVelScale=<0.35 0.35 0.35>
    gravityOverride {z=-980}
    physicsBodyDefinitions [{
      sphereCollisionRadius=0
      boundBone {boneName=dyn_hairTips_02}
      boxExtents=<15 10 10>
      constraintSetup {angularLimitsMax=<0 2 20>, angularLimitsMin=<0 -35 -20>, angularTarget=<0 0 0>}
      localJointOffset=<2 0 -3>
     }]
    relativeSpaceBone {boneName=head}
   }
   n3 = Anim Dynamics - Bone: dyn_hairTips_03(ComponentPose=@n2, Alpha=1, ExternalForce="0,0,0") {
    angularDampingOverride=0.92
    angularSpringConstant=1000
    bAngularSpring=true
    bGravityOverrideInSimSpace=true
    bOverrideAngularDamping=true
    bUseGravityOverride=true
    simulationSpace=RootRelative
    alphaBoolBlend {blendOption=Linear}
    boundBone {boneName=dyn_hairTips_03}
    componentAppliedLinearAccClamp=<500 500 500>
    componentLinearAccScale=<0.4 0.4 0.4>
    componentLinearVelScale=<0.35 0.35 0.35>
    gravityOverride {z=-980}
    physicsBodyDefinitions [{
      sphereCollisionRadius=0
      boundBone {boneName=dyn_hairTips_03}
      boxExtents=<12 15 20>
      constraintSetup {angularLimitsMax=<0 2 20>, angularLimitsMin=<0 -35 -20>, angularTarget=<0 0 0>}
      localJointOffset=<2 0 -3>
     }]
    relativeSpaceBone {boneName=head}
   }
   n4 = Anim Dynamics - Bone: dyn_hairTips_04(ComponentPose=@n3, Alpha=1, ExternalForce="0,0,0") {
    angularDampingOverride=0.92
    angularSpringConstant=800
    bAngularSpring=true
    bGravityOverrideInSimSpace=true
    bOverrideAngularDamping=true
    bUseGravityOverride=true
    simulationSpace=RootRelative
    alphaBoolBlend {blendOption=Linear}
    boundBone {boneName=dyn_hairTips_04}
    componentAppliedLinearAccClamp=<500 500 500>
    componentLinearAccScale=<0.4 0.4 0.4>
    componentLinearVelScale=<0.35 0.35 0.35>
    gravityOverride {z=-980}
    physicsBodyDefinitions [{
      sphereCollisionRadius=0
      boundBone {boneName=dyn_hairTips_04}
      boxExtents=<10 14 14>
      constraintSetup {angularLimitsMax=<0 2 20>, angularLimitsMin=<0 -35 -20>, angularTarget=<0 0 0>}
      localJointOffset=<1.5 0 0>
     }]
    relativeSpaceBone {boneName=head}
   }
   Anim Dynamics - Bone: dyn_hairTips_05(ComponentPose=@n4, Alpha=1, ExternalForce="0,0,0") {
    angularDampingOverride=0.92
    bAngularSpring=true
    bGravityOverrideInSimSpace=true
    bOverrideAngularDamping=true
    bUseGravityOverride=true
    simulationSpace=RootRelative
    alphaBoolBlend {blendOption=Linear}
    boundBone {boneName=dyn_hairTips_05}
    componentAppliedLinearAccClamp=<500 500 500>
    componentLinearAccScale=<0.4 0.4 0.4>
    componentLinearVelScale=<0.35 0.35 0.35>
    gravityOverride {z=-980}
    physicsBodyDefinitions [{
      sphereCollisionRadius=0
      boundBone {boneName=dyn_hairTips_05}
      boxExtents=<10 10 8>
      constraintSetup {angularLimitsMax=<0 2 20>, angularLimitsMin=<0 -35 -20>, angularTarget=<0 0 0>}
      localJointOffset=<1.5 0 0>
     }]
    relativeSpaceBone {boneName=head}
   }
   n5 = RigidBody!(
    ComponentPose=Local To Component(LocalPose=Use cached pose 'prePose'())
    Alpha
    OverrideWorldGravity:Vector=Get GravityOverride()
    ExternalForce
   ) -> (Pose) {
    bClampLinearTranslationLimitToRefPose=true
    bOverrideWorldGravity=true
    overridePhysicsAsset=&/F_MED_Dino_Collector_dyn_hair
    simulationSpace=BaseBoneSpace
    alphaBoolBlend {blendOption=Linear}
    baseBoneRef {boneName=spine_05}
    componentAppliedLinearAccClamp=<500 0 0>
    componentLinearAccScale {x=0.4}
    componentLinearVelScale {x=0.4}
    overrideWorldGravity {z=-980}
    simSpaceSettings {
     maxAngularAcceleration=100
     maxAngularVelocity=5
     maxLinearAcceleration=500
     maxLinearVelocity=50
     velocityScaleZ=0
     worldAlpha=1
     externalLinearDragV=<1 1>
    }
   }
   n6 = Look At - Bone: dyn_templeBraid_01_r(ComponentPose=Local To Component(LocalPose=Use cached pose 'prePose'()), Alpha=1) {
    alphaBoolBlend {blendOption=Linear}
    boneToModify {boneName=dyn_templeBraid_01_r}
    lookAtLocation=<0 5 7>
    lookAtTarget {boneReference {boneName=spine_05}}
    lookAt_Axis {axis=<1 0>}
   }
   n7 = Look At - Bone: dyn_templeBraid_03_r(ComponentPose=@n6, Alpha=1) {
    alphaBoolBlend {blendOption=Linear}
    boneToModify {boneName=dyn_templeBraid_03_r}
    lookAtLocation=<0 10 9>
    lookAtTarget {boneReference {boneName=spine_05}}
    lookAt_Axis {axis=<1 0>}
   }
   n8 = Look At - Bone: dyn_templeBraid_01_l(ComponentPose=@n7, Alpha=1) {
    alphaBoolBlend {blendOption=Linear}
    boneToModify {boneName=dyn_templeBraid_01_l}
    lookAtLocation=<0 5 -8>
    lookAtTarget {boneReference {boneName=spine_05}}
    lookAt_Axis {axis=<1 0>}
   }
   n9 = Look At - Bone: dyn_templeBraid_03_l(ComponentPose=@n8, Alpha=1) {
    alphaBoolBlend {blendOption=Linear}
    boneToModify {boneName=dyn_templeBraid_03_l}
    lookAtLocation=<0 9 -9>
    lookAtTarget {boneReference {boneName=spine_05}}
    lookAt_Axis {axis=<1 0>}
   }
   n10 = Look At - Bone: dyn_hair_bk_02(ComponentPose=@n9, Alpha=1) {
    alphaBoolBlend {blendOption=Linear}
    boneToModify {boneName=dyn_hair_bk_02}
    lookAtLocation=<0 -11>
    lookAtTarget {boneReference {boneName=spine_05}}
    lookAt_Axis {axis=<1 0>}
    lookUp_Axis {axis {x=0, z=1}}
   }
   n11 = Look At - Bone: dyn_hair_bk_l_00(ComponentPose=@n10, Alpha=1) {
    alphaBoolBlend {blendOption=Linear}
    boneToModify {boneName=dyn_hair_bk_l_00}
    lookAtLocation=<0 -11 -4>
    lookAtTarget {boneReference {boneName=spine_05}}
    lookAt_Axis {axis=<1 0>}
    lookUp_Axis {axis {x=0, z=1}}
   }
   n12 = Look At - Bone: dyn_hair_bk_r_00(ComponentPose=@n11, Alpha=1) {
    alphaBoolBlend {blendOption=Linear}
    boneToModify {boneName=dyn_hair_bk_r_00}
    lookAtLocation=<0 -11 4>
    lookAtTarget {boneReference {boneName=spine_05}}
    lookAt_Axis {axis=<1 0>}
    lookUp_Axis {axis {x=0, z=1}}
   }
   Output Pose(Result=Blend Poses by bool(BlendPose_0=Component To Local(ComponentPose=@n5), BlendPose_1=Component To Local(ComponentPose=@n12), bActiveValue=Get bRigidBodyEnabled()) {blendType=Linear}) {layerGroup=None}
   n13 = Copy Pose From Mesh(SourceMeshComponent=Get Body_Mesh())
   n14 = Copy Pose From Mesh() {bUseAttachedParent=true}
   n15 = Look At - Bone: dyn_hair_bk_MSTR(
    ComponentPose=Local To Component(LocalPose=Layered blend per bone(BasePose=@n14, BlendPoses_0=@n13) {layerSetup [{branchFilters [{boneName=upperarm_l}, {boneName=upperarm_r}]}]})
    Alpha=1
   ) {
    alphaBoolBlend {blendOption=Linear}
    boneToModify {boneName=dyn_hair_bk_MSTR}
    lookAtLocation=<0 -8>
    lookAtTarget {boneReference {boneName=spine_05}}
    lookAt_Axis {axis=<1 0>}
    lookUp_Axis {axis {x=0, z=1}}
   }
   n16 = "Transform (Modify) Bone - Bone: dyn_hair_bk_MSTR"(ComponentPose=@n15, Alpha=Get head_rot_bk() -> (head_rot_bk:double), Translation="-1.842,-0.000,1.120", Rotation="0,0,0", Scale="1,1,1") {
    rotationMode=BMM_Additive
    rotationSpace=BCS_BoneSpace
    translationMode=BMM_Additive
    translationSpace=BCS_BoneSpace
    alphaBoolBlend {blendOption=Linear}
    boneToModify {boneName=dyn_hair_bk_MSTR}
    translation {x=-1.842, z=1.12}
   }
   Save cached pose 'prePose'(Pose=Component To Local(ComponentPose=@n16)) {cacheName=prePose}
  ]}
 EventGraph {nodes [
   Event BlueprintInitializeAnimation()
   > Set Body_Mesh(Body_Mesh=GetOwnerSkeletalMeshComponent())
   Event BlueprintUpdateAnimation()
   n1 = BreakTransform(InTransform=GetDeltaTransformFromRefPose(self=GetOwnerSkeletalMeshComponent(), BoneName=head, BaseName=spine_05))
   > Set head_rot_bk(head_rot_bk:double=MapRangeClamped(Value=BreakRotator(InRot=@n1.Rotation).Yaw, InRangeA=0, InRangeB=-50)) -> (Output_Get:double)
   > n2 = Gravity Override State Function DEPRECATED(CharacterPartAnimInstance=Self-Reference(), GameTimeInSeconds=GetGameTimeInSeconds(), GravityOverrideParams=Get Gravity Override Params())
   > Set GravityOverride(GravityOverride=@n2)
  ]}
]
variables [
 Body_Mesh {type=SkeletalMeshComponent}
 Gravity Override Params {
  type=GravityOverrideParamsStruct
  default {
   skydive_DivingDown {gravityOverride=<-500 0 0>, windAmplitude=<50 100 50>, windFrequency=<1 3 2>}
   skydive_FloatingDown {gravityOverride=<0 -100 -300>, windAmplitude=<200 -100 100>, windOffset=<0 -200 100>}
   skydive_GoingUp {gravityOverride=<0 -100 -100>, windAmplitude=<100 -100 50>, windFrequency=<0.5 3 4>}
   skydive_Parachuting {gravityOverride=<0 -500 -980>, windAmplitude=<200 -300 400>}
  }
 }
 GravityOverride {type=Vector}
 head_rot_bk {type=number}
]
